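# Niagara system for the big lightning effect in the Total Flock emote, built as one packed emitter.
# Three bursts spawn a glow, a few smaller randomized mesh bolts and a large bolt, with spawn-group masks setting each one's look and a visibility tag.
# ElimColor tints particles, while curves drive flicker, scale and dynamic material parameters across sprite and mesh renderers.
# Sprite color also scales by SMTOD, with lower alpha in front-end preview.
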NiagaraSystem /BRCosmetics/Animation/Game/MainPlayer/Emotes/TotalFlock/FX/NS_Emote_TotalFlock_BigLightning
dependencies {
 usedDataInterfaces [&NiagaraDataInterfaceVectorCurve, &NiagaraDataInterfaceCurve, &NiagaraDataInterfaceArrayColor]
 usedDynamicInputs [
  &UniformRangedVector
  &InvertBool
  &SelectLinearColorFromArray
 ]
 usedModules [
  &SystemState
  &self:SystemSpawnScript.SystemScriptSource.SystemScriptGraph.NiagaraNodeAssignment_0.SetVariables_#1
  &self:SystemSpawnScript.SystemScriptSource.SystemScriptGraph.NiagaraNodeAssignment_1.SetVariables_#2
 ]
}
emitters {BigLightning_Packed {
  emitterUpdateScript [
   &EmitterState()
   &SpawnBurst_Instantaneous(Spawn Group=2, Use Spawn Probability=false, UseLoopCountLimit=false) {module=SpawnBurst_Instantaneous002}
   &SpawnBurst_Instantaneous(
    Spawn Count=&RandomRangeInt(Minimum=2, Maximum=3)
    Spawn Group=1
    Use Spawn Probability=false
    UseLoopCountLimit=false
   ) {module=SpawnBurst_Instantaneous001}
   &SpawnBurst_Instantaneous(Spawn Count=1, Use Spawn Probability=false)
  ]
  particleSpawnScript [
   /SetVariables(Particles.VisibilityTag=&SpawnGroup()) {isSetParameters=true}
   &InitializeParticle(
    Color=&MaskLinearColorBySpawnGroup(
     Masked Linear Color=&Multiply_LinearColorByFloat(Linear Color=$System.ElimColor, Float=5)
     Passthrough Linear Color=&MaskLinearColorBySpawnGroup(
      Masked Linear Color=&Multiply_LinearColorByFloat(Linear Color=$System.ElimColor, Float=0.4)
      Passthrough Linear Color=&Multiply_LinearColorByFloat(Linear Color=$System.ElimColor, Float=0.4)
      Spawn Group Mask=1
     )
    )
    Lifetime=&MaskFloatBySpawnGroup(
     Masked Float=0.575
     Passthrough Float=&MaskFloatBySpawnGroup(Masked Float=0.4, Passthrough Float=0.5, Spawn Group Mask=1)
    )
    Mass Mode=::NewEnumerator1
    Mesh Scale Max=&MaskVectorBySpawnGroup(
     Masked Vector=<0 0 0>
     Passthrough Vector=&MaskVectorBySpawnGroup(Masked Vector=<1.5 1.5 6>, Passthrough Vector=<4 4 15>, Spawn Group Mask=1)
    )
    Mesh Scale Min=&MaskVectorBySpawnGroup(
     Masked Vector=<0 0 0>
     Passthrough Vector=&MaskVectorBySpawnGroup(Masked Vector=<1 1 3>, Passthrough Vector=<4 4 15>, Spawn Group Mask=1)
    )
    Mesh Scale Mode=::NewEnumerator2
    Position=&MaskVectorBySpawnGroup(
     Masked Vector=<10 -5 350>
     Passthrough Vector=&MaskVectorBySpawnGroup(
      Masked Vector=&RandomRangeVector(Minimum=<5 0 150>, Maximum=<5 0 350>)
      Passthrough Vector=<10 -5 350>
      Spawn Group Mask=1
     )
    )
    Position Mode=::NewEnumerator1
    Sprite Size=<120 85>
    Sprite Size Mode=::NewEnumerator4
    UsePositionOffset=false
    Write Material Random=true
   ) {module=InitializeParticle001}
   &InitialMeshOrientation(
    Mesh Orientation Mode=::NewEnumerator2
    Rotation=&MaskVectorBySpawnGroup(
     Masked Vector=<0 0 0>
     Passthrough Vector=&MaskVectorBySpawnGroup(
      Masked Vector=&RandomRangeVector(Minimum=<0.49 -0.15 -0.15>, Maximum=<0.51 0.15 0.15>)
      Passthrough Vector=<0.5 0 0>
      Spawn Group Mask=1
     )
    )
    Use Rotation Vector=true
   )
  ]
  particleUpdateScript [
   &SolveForcesAndVelocity(Clamp Velocity=false, Limit Acceleration=false)
   &ParticleState()
   &ScaleMeshSize(Scale Factor=&MaskVectorBySpawnGroup(
     Masked Vector=<0 0 0>
     Passthrough Vector=&MaskVectorBySpawnGroup(
      Masked Vector=&Multiply_VectorByFloat(Vector=<1 1 1>, Float=0.2)
      Passthrough Vector=&Multiply_VectorByFloat(
       Vector=&VectorFromCurve(
        VectorCurve=DataInterface() {propertyValues {
          XCurve {keys [<at 0 1>,<at 0.3 1.4>,<at 0.5 1>,<at 1 1>]}
          YCurve {keys [<at 0 1>,<at 0.3 1.4>,<at 0.5 1>,<at 1 1>]}
          ZCurve {keys [<at 0 1>,<at 1 1>]}
         }}
        Scale Curve=<1.3 1.3 1>
       )
       Float=0.2
      )
      Spawn Group Mask=1
     )
    ))
   /SetVariables(
    Particles.LensFlareColor:LinearColor=&Multiply_LinearColorByFloat(
     Linear Color=<rgba 0 4.4118 15 1>
     Float=&FloatFromCurve(
      FloatCurve=DataInterface() {propertyValues {Curve {keys [
          {arriveTangent=16.3847, interpMode=RCIM_Cubic, leaveTangent=16.3847, tangentMode=RCTM_User, time=0, value=0}
          {arriveTangent=2.6753, interpMode=RCIM_Cubic, leaveTangent=2.6753, time=0.0145, value=0.9}
          {arriveTangent=-1.2401, interpMode=RCIM_Cubic, leaveTangent=-1.2401, tangentMode=RCTM_User, time=0.3925, value=1.05}
          {arriveTangent=-0.7226, tangentMode=RCTM_User, time=0.4514, value=0.8}
          <at 0.5342 0.9>
          <at 0.6086 0.55>
          {arriveTangent=-11.4091, interpMode=RCIM_Cubic, leaveTangent=-11.4091, tangentMode=RCTM_User, time=0.7, value=0.9}
          <at 1 0>
         ]}}}
      Scale Curve=0.15
     )
    )
    Particles.LensFlareSize:Vector2f=&Multiply_Vector2D(A=$Particles.SpriteSize, B=<0.6 1>)
   ) {isSetParameters=true}
   /SetVariables(Particles.StarColor:LinearColor=&Multiply_LinearColorByFloat(
     Linear Color=$Particles.Color
     Float=&FloatFromCurve(
      FloatCurve=DataInterface() {propertyValues {Curve {keys [
          <at 0 1 RCIM_Cubic>
          {arriveTangent=-1.4569, interpMode=RCIM_Cubic, leaveTangent=-1.4569, tangentMode=RCTM_User, time=0.6, value=0.7}
          {arriveTangent=-0.1457, interpMode=RCIM_Cubic, leaveTangent=-0.1457, tangentMode=RCTM_User, time=1, value=0}
         ]}}}
      Scale Curve=0.075
     )
    )) {isSetParameters=true}
   &ScaleColor(
    Scale Alpha=&MaskFloatBySpawnGroup(
     Masked Float=&FloatFromCurve(FloatCurve=DataInterface() {propertyValues {Curve {keys [
          {arriveTangent=16.3847, interpMode=RCIM_Cubic, leaveTangent=16.3847, tangentMode=RCTM_User, time=0, value=0}
          {arriveTangent=3.3073, interpMode=RCIM_Cubic, leaveTangent=3.3073, time=0.0145, value=0.9}
          {arriveTangent=-1.2401, interpMode=RCIM_Cubic, leaveTangent=-1.2401, tangentMode=RCTM_User, time=0.3326, value=1.1}
          {arriveTangent=-0.7226, tangentMode=RCTM_User, time=0.385, value=0.75}
          <at 0.4743 0.95>
          <at 0.5401 0.45>
          {arriveTangent=0.131, interpMode=RCIM_Cubic, leaveTangent=0.131, time=0.6979, value=0.75}
          {arriveTangent=-2.4823, interpMode=RCIM_Cubic, leaveTangent=-2.4823, time=0.9219, value=0.5}
          <at 1 0>
         ]}}})
     Passthrough Float=&MaskFloatBySpawnGroup(
      Masked Float=&FloatFromCurve(FloatCurve=DataInterface() {propertyValues {Curve {keys [
           {tangentMode=RCTM_User, time=0, value=1}
           {arriveTangent=-0.0405, interpMode=RCIM_Cubic, leaveTangent=-0.0405, tangentMode=RCTM_User, time=0.7009, value=1}
           {arriveTangent=-8.0089, interpMode=RCIM_Cubic, leaveTangent=-8.0089, tangentMode=RCTM_User, time=1, value=0}
          ]}}})
      Passthrough Float=&FloatFromCurve(FloatCurve=DataInterface() {propertyValues {Curve {keys [
           {tangentMode=RCTM_User, time=0, value=1}
           {arriveTangent=-0.0405, interpMode=RCIM_Cubic, leaveTangent=-0.0405, tangentMode=RCTM_User, time=0.7009, value=1}
           {arriveTangent=-8.0089, interpMode=RCIM_Cubic, leaveTangent=-8.0089, tangentMode=RCTM_User, time=1, value=0}
          ]}}})
      Spawn Group Mask=1
     )
    )
    Scale RGB=&MaskVectorBySpawnGroup(
     Masked Vector=&Lerp_Vector(
      A=&VectorFromFloat(Value=&FloatFromCurve(
        FloatCurve=DataInterface() {propertyValues {Curve {keys [<at 0 0>,<at 0.1 5>,<at 0.5 3.75>,<at 0.6 2.5>,<at 0.7 4.25>,<at 0.8 1>,<at 0.9 1.15>,<at 1 1>]}}}
        Scale Curve=10
       ))
      B=&VectorFromFloat(Value=&RandomRangeFloat(Minimum=0.5, Maximum=1, Evaluation Type=::NewEnumerator1))
      Alpha=&FloatFromCurve(FloatCurve=DataInterface() {propertyValues {Curve {keys [<at 0 0 RCIM_Cubic>,<at 1 1 RCIM_Cubic>]}}})
     )
     Passthrough Vector=&MaskVectorBySpawnGroup(
      Masked Vector=&Lerp_Vector(
       A=&VectorFromFloat(Value=&FloatFromCurve(
         FloatCurve=DataInterface() {propertyValues {Curve {keys [<at 0 1>,<at 0.1 1>,<at 0.2 5>,<at 0.3 1>,<at 0.6 1>,<at 0.7 2>,<at 0.8 1>,<at 1 1>]}}}
         Scale Curve=10
        ))
       B=&VectorFromFloat(Value=&RandomRangeFloat(Minimum=0.3, Maximum=5, Evaluation Type=::NewEnumerator1))
       Alpha=&FloatFromCurve(FloatCurve=DataInterface() {propertyValues {Curve {keys [
            {interpMode=RCIM_Cubic, leaveTangent=2.5294, tangentMode=RCTM_Break, time=0, value=0}
            {arriveTangent=-0.2166, interpMode=RCIM_Cubic, tangentMode=RCTM_Break, time=0.4, value=1}
           ]}}})
      )
      Passthrough Vector=&Lerp_Vector(
       A=&VectorFromFloat(Value=&FloatFromCurve(
         FloatCurve=DataInterface() {propertyValues {Curve {keys [<at 0 1>,<at 0.1 1>,<at 0.2 5>,<at 0.3 1>,<at 0.6 1>,<at 0.7 2>,<at 0.8 1>,<at 1 1>]}}}
         Scale Curve=10
        ))
       B=&VectorFromFloat(Value=&RandomRangeFloat(Minimum=0.3, Maximum=5, Evaluation Type=::NewEnumerator1))
       Alpha=&FloatFromCurve(FloatCurve=DataInterface() {propertyValues {Curve {keys [
            {interpMode=RCIM_Cubic, leaveTangent=2.5294, tangentMode=RCTM_Break, time=0, value=0}
            {arriveTangent=-0.2166, interpMode=RCIM_Cubic, tangentMode=RCTM_Break, time=0.4, value=1}
           ]}}})
      )
      Spawn Group Mask=1
     )
    )
    ScaleA=true
    ScaleRGB=true
   )
   &ScaleSpriteSize(Uniform Curve Sprite Scale=DataInterface() {propertyValues {Curve {keys [
        <at 0 0.5 RCIM_Cubic>
        {arriveTangent=0.7134, interpMode=RCIM_Cubic, leaveTangent=0.7134, time=0.1, value=1}
        {arriveTangent=-0.0405, interpMode=RCIM_Cubic, leaveTangent=-0.0405, tangentMode=RCTM_User, time=0.7009, value=1}
        {arriveTangent=-8.0089, interpMode=RCIM_Cubic, leaveTangent=-8.0089, tangentMode=RCTM_User, time=1, value=0}
       ]}}})
   /SetVariables(Particles.SpriteColor=&MultiplyLinearColor(
     Linear Color=$Particles.Color
     Scale Factor=&MakeVector4(
      X=$System.SMTOD
      Y=$System.SMTOD
      Z=$System.SMTOD
      W=&MakeCustomFloatFromBool(Bool=$User.bIsFrontEndPreview, True Float=0.06, False Float=0.13)
     )
    )) {isSetParameters=true}
   &UpdateMeshOrientation(Rotation Rate=&RandomRangeFloat(Minimum=-1, Maximum=1), Rotation Vector=<0 0 1>)
   &DynamicMaterialParameters(
    Index 0 Param 1=&MaskFloatBySpawnGroup(
     Masked Float=0
     Passthrough Float=&MaskFloatBySpawnGroup(
      Masked Float=&MakeFloatFromVector(Vector=$Particles.Scale, Channel=::NewEnumerator0)
      Passthrough Float=&MakeFloatFromVector(Vector=$Particles.Scale, Channel=::NewEnumerator0)
      Spawn Group Mask=1
     )
    )
    Index 0 Param 2=&MaskFloatBySpawnGroup(
     Masked Float=0
     Passthrough Float=&MaskFloatBySpawnGroup(
      Masked Float=&RandomRangeFloat(Minimum=0, Maximum=10)
      Passthrough Float=&RandomRangeFloat(Minimum=4, Maximum=4)
      Spawn Group Mask=1
     )
    )
    Index 0 Param 3=&MaskFloatBySpawnGroup(
     Masked Float=0
     Passthrough Float=&MaskFloatBySpawnGroup(
      Masked Float=&FloatFromCurve(
       FloatCurve=DataInterface() {propertyValues {Curve {keys [{interpMode=RCIM_Cubic, leaveTangent=8.1085, tangentMode=RCTM_Break, time=0.8568, value=1}, {tangentMode=RCTM_User, time=1, value=0}]}}}
       Scale Curve=0.07
      )
      Passthrough Float=&FloatFromCurve(
       FloatCurve=DataInterface() {propertyValues {Curve {keys [{interpMode=RCIM_Cubic, leaveTangent=8.1085, tangentMode=RCTM_Break, time=0.8568, value=1}, {tangentMode=RCTM_User, time=1, value=0}]}}}
       Scale Curve=0.1
      )
      Spawn Group Mask=1
     )
    )
    Index 0 Param 4=&MaskFloatBySpawnGroup(
     Masked Float=0
     Passthrough Float=&MaskFloatBySpawnGroup(Masked Float=0.8, Passthrough Float=0.8, Spawn Group Mask=1)
    )
    Index 1 Param 1=&MaskFloatBySpawnGroup(
     Masked Float=0
     Passthrough Float=&MaskFloatBySpawnGroup(
      Masked Float=&FloatFromCurve(FloatCurve=DataInterface() {propertyValues {Curve {keys [
           {arriveTangent=7.4517, interpMode=RCIM_Cubic, leaveTangent=7.4517, tangentMode=RCTM_User, time=0, value=0}
           {arriveTangent=1.1179, interpMode=RCIM_Cubic, leaveTangent=1.1179, tangentMode=RCTM_User, time=0.1419, value=0.65}
           {arriveTangent=0.0224, interpMode=RCIM_Cubic, leaveTangent=0.0224, tangentMode=RCTM_User, time=1, value=1}
          ]}}})
      Passthrough Float=&FloatFromCurve(FloatCurve=DataInterface() {propertyValues {Curve {keys [
           {arriveTangent=4.8997, interpMode=RCIM_Cubic, leaveTangent=4.8997, tangentMode=RCTM_User, time=0, value=0}
           {arriveTangent=2.0405, interpMode=RCIM_Cubic, leaveTangent=2.0405, tangentMode=RCTM_User, time=0.2, value=0.65}
           {arriveTangent=-0.8125, interpMode=RCIM_Cubic, leaveTangent=-0.8125, time=0.7, value=0.95}
           {arriveTangent=-3.9378, interpMode=RCIM_Cubic, leaveTangent=-3.9378, tangentMode=RCTM_User, time=1, value=0}
          ]}}})
      Spawn Group Mask=1
     )
    )
    Index 1 Param 2=&MaskFloatBySpawnGroup(
     Masked Float=0
     Passthrough Float=&MaskFloatBySpawnGroup(
      Masked Float=&FloatFromCurve(FloatCurve=DataInterface() {propertyValues {Curve {keys [<at 0.5 0>, {arriveTangent=3.2053, tangentMode=RCTM_User, time=0.9, value=1}]}}})
      Passthrough Float=&FloatFromCurve(FloatCurve=DataInterface() {propertyValues {Curve {keys [
           {interpMode=RCIM_Cubic, tangentMode=RCTM_User, time=0, value=0}
           {arriveTangent=1.7949, interpMode=RCIM_Cubic, leaveTangent=1.7949, tangentMode=RCTM_User, time=1, value=1}
          ]}}})
      Spawn Group Mask=1
     )
    )
    Index 1 Param 3=&MaskFloatBySpawnGroup(
     Masked Float=0
     Passthrough Float=&MaskFloatBySpawnGroup(
      Masked Float=&FloatFromCurve(FloatCurve=DataInterface() {propertyValues {Curve {keys [<at 0 0 RCIM_Cubic>, {arriveTangent=3.4223, interpMode=RCIM_Cubic, leaveTangent=3.4223, tangentMode=RCTM_User, time=1, value=1}]}}})
      Passthrough Float=&FloatFromCurve(FloatCurve=DataInterface() {propertyValues {Curve {keys [<at 0 0 RCIM_Cubic>, {arriveTangent=3.4223, interpMode=RCIM_Cubic, leaveTangent=3.4223, tangentMode=RCTM_User, time=1, value=1}]}}})
      Spawn Group Mask=1
     )
    )
    Index 1 Param 4=&MaskFloatBySpawnGroup(
     Masked Float=0
     Passthrough Float=&MaskFloatBySpawnGroup(
      Masked Float=&FloatFromCurve(
       FloatCurve=DataInterface() {propertyValues {Curve {keys [
           {interpMode=RCIM_Cubic, leaveTangent=0.1169, tangentMode=RCTM_Break, time=0, value=0}
           {arriveTangent=2.5708, interpMode=RCIM_Cubic, leaveTangent=2.5708, tangentMode=RCTM_User, time=1, value=1}
          ]}}}
       Scale Curve=&RandomRangeFloat(Minimum=1, Maximum=40)
      )
      Passthrough Float=&FloatFromCurve(
       FloatCurve=DataInterface() {propertyValues {Curve {keys [
           {interpMode=RCIM_Cubic, leaveTangent=0.1169, tangentMode=RCTM_Break, time=0, value=0}
           {arriveTangent=2.5708, interpMode=RCIM_Cubic, leaveTangent=2.5708, tangentMode=RCTM_User, time=1, value=1}
          ]}}}
       Scale Curve=&RandomRangeFloat(Minimum=1, Maximum=2)
      )
      Spawn Group Mask=1
     )
    )
    Index 2 Param 1=&MaskFloatBySpawnGroup(
     Masked Float=0
     Passthrough Float=&MaskFloatBySpawnGroup(
      Masked Float=&FloatFromCurve(FloatCurve=DataInterface() {propertyValues {Curve {keys [
           {interpMode=RCIM_Cubic, leaveTangent=-0.0025, tangentMode=RCTM_Break, time=0, value=0.5}
           {arriveTangent=-0.0316, interpMode=RCIM_Cubic, tangentMode=RCTM_Break, time=1, value=-0.2}
          ]}}})
      Passthrough Float=&FloatFromCurve(FloatCurve=DataInterface() {propertyValues {Curve {keys [{interpMode=RCIM_Cubic, tangentMode=RCTM_User, time=0, value=1}, {interpMode=RCIM_Cubic, tangentMode=RCTM_User, time=1, value=0}]}}})
      Spawn Group Mask=1
     )
    )
    Index 2 Param 2=&MaskFloatBySpawnGroup(
     Masked Float=0
     Passthrough Float=&MaskFloatBySpawnGroup(
      Masked Float=&FloatFromCurve(
       FloatCurve=DataInterface() {propertyValues {Curve {keys [
           {interpMode=RCIM_Cubic, tangentMode=RCTM_User, time=0, value=0}
           {arriveTangent=2.0522, interpMode=RCIM_Cubic, leaveTangent=2.0522, tangentMode=RCTM_User, time=1, value=1}
          ]}}}
       Scale Curve=2
      )
      Passthrough Float=&FloatFromCurve(
       FloatCurve=DataInterface() {propertyValues {Curve {keys [
           {interpMode=RCIM_Cubic, tangentMode=RCTM_User, time=0, value=0}
           {arriveTangent=2.0522, interpMode=RCIM_Cubic, leaveTangent=2.0522, tangentMode=RCTM_User, time=1, value=1}
          ]}}}
       Scale Curve=2
      )
      Spawn Group Mask=1
     )
    )
    Index 2 Param 3=&MaskFloatBySpawnGroup(
     Masked Float=0
     Passthrough Float=&MaskFloatBySpawnGroup(Masked Float=0, Passthrough Float=1, Spawn Group Mask=1)
    )
    Index 2 Param 4=&MaskFloatBySpawnGroup(
     Masked Float=0
     Passthrough Float=&MaskFloatBySpawnGroup(Masked Float=0, Passthrough Float=1, Spawn Group Mask=1)
    )
    Index 3 Param 1=&MaskFloatBySpawnGroup(
     Masked Float=0
     Passthrough Float=&MaskFloatBySpawnGroup(Masked Float=0, Passthrough Float=1, Spawn Group Mask=1)
    )
    Index 3 Param 2=&MaskFloatBySpawnGroup(
     Masked Float=0
     Passthrough Float=&MaskFloatBySpawnGroup(Masked Float=0, Passthrough Float=1, Spawn Group Mask=1)
    )
    Index 3 Param 3=&MaskFloatBySpawnGroup(
     Masked Float=0
     Passthrough Float=&MaskFloatBySpawnGroup(Masked Float=0, Passthrough Float=1, Spawn Group Mask=1)
    )
    Index 3 Param 4=&MaskFloatBySpawnGroup(
     Masked Float=0
     Passthrough Float=&MaskFloatBySpawnGroup(Masked Float=0, Passthrough Float=0, Spawn Group Mask=1)
    )
    Module.Index0Float0Write=true
    Module.Index0Float1Write=true
    Module.Index0Float2Write=true
    Module.Index0Float3Write=true
    Module.Index1Float0Write=true
    Module.Index1Float1Write=true
    Module.Index1Float2Write=true
    Module.Index1Float3Write=true
    Module.Index2Float0Write=true
    Module.Index2Float1Write=true
    Module.Index2Float2Write=false
    Module.Index2Float3Write=false
    Module.Index3Float0Write=false
    Module.Index3Float1Write=false
    Module.Index3Float2Write=false
    Module.Index3Float3Write=true
    Write Parameter Index 1=::NewEnumerator1
    Write Parameter Index 2=::NewEnumerator1
    Write Parameter Index 3=::NewEnumerator1
   )
  ]
  rendererClasses [&NiagaraSpriteRendererProperties, &NiagaraMeshRendererProperties]
  renderers [
   {rendererClass=&NiagaraSpriteRendererProperties}
   {rendererClass=&NiagaraSpriteRendererProperties, rendererIndex=1}
   {rendererClass=&NiagaraSpriteRendererProperties, rendererIndex=2}
   {rendererClass=&NiagaraMeshRendererProperties, rendererIndex=3}
   {rendererClass=&NiagaraMeshRendererProperties, rendererIndex=4}
   {rendererClass=&NiagaraMeshRendererProperties, rendererIndex=5}
   {rendererClass=&NiagaraMeshRendererProperties, rendererIndex=6}
  ]
 }}
summary {
 systemName=NS_Emote_TotalFlock_BigLightning
 emitters [{
   emitterName=BigLightning_Packed
   rendererClasses [&NiagaraSpriteRendererProperties, &NiagaraMeshRendererProperties]
  }]
 userVariables [
  User.bIsFrontEndPreview {type=NiagaraBool}
  User.bKillStage {type=NiagaraInt32}
 ]
}
