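# Niagara effect for the medallion in the TotalFlock emote, layering a glowing medallion burst with flares, lightning and flying particles.
# One emitter mixes timed bursts and spawn rates in six spawn groups, using nested spawn-group masks to give each its own look and motion.
# The spawn group becomes a visibility tag for nine sprite and mesh renderers.
# System ElimColor drives tint, while curves animate scale, brightness and dynamic material parameters.
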NiagaraSystem /BRCosmetics/Animation/Game/MainPlayer/Emotes/TotalFlock/FX/NS_Emote_TotalFlock_Medallion
dependencies {
 usedDataInterfaces [
  &NiagaraDataInterfaceCurve
  &NiagaraDataInterfaceVector2DCurve
  &NiagaraDataInterfaceVectorCurve
  &NiagaraDataInterfaceArrayColor
 ]
 usedDynamicInputs [&UniformRangedVector, &SelectLinearColorFromArray]
 usedModules [
  &SystemState
  &self:SystemSpawnScript.SystemScriptSource.SystemScriptGraph.NiagaraNodeAssignment_0.SetVariables_#1
  &self:SystemSpawnScript.SystemScriptSource.SystemScriptGraph.NiagaraNodeAssignment_1.SetVariables_#2
  &self:SystemSpawnScript.SystemScriptSource.SystemScriptGraph.NiagaraNodeAssignment_3.SetVariables_#3
  &self:SystemSpawnScript.SystemScriptSource.SystemScriptGraph.NiagaraNodeAssignment_2.SetVariables_#4
 ]
}
emitters {Medallion {
  emitterUpdateScript [
   &EmitterState(Life Cycle Mode=::NewEnumerator1, Loop Behavior=::NewEnumerator1, Loop Duration=$System.CurrentLoopDuration, UseLoopDelay=false)
   &SpawnBurst_Instantaneous(Spawn Count=1, Spawn Time=$System.Delay, Use Spawn Probability=false)
   &SpawnBurst_Instantaneous(Spawn Group=1, Spawn Time=$System.Delay, Use Spawn Probability=false, UseLoopCountLimit=false) {module=SpawnBurst_Instantaneous001}
   &SpawnRate(
    SpawnGroup=2
    SpawnRate=&Multiply_Float(
     A=50
     B=&FloatFromCurve(
      FloatCurve=DataInterface() {propertyValues {Curve {keys [
          <at 0 0 RCIM_Constant>
          {tangentMode=RCTM_User, time=0.2, value=1.075}
          {interpMode=RCIM_Cubic, tangentMode=RCTM_User, time=0.4, value=1.025}
          {arriveTangent=-7.5091, interpMode=RCIM_Cubic, leaveTangent=-7.5091, tangentMode=RCTM_User, time=0.8, value=0.025}
         ]}}}
      CurveIndex=$Emitter.NormalizedLoopAge
     )
    )
    Use Spawn Probability=false
   )
   &SpawnBurst_Instantaneous(Spawn Count=50, Spawn Group=2, Spawn Time=$System.Delay, Use Spawn Probability=false, UseLoopCountLimit=false) {module=SpawnBurst_Instantaneous002}
   &SpawnRate(
    Spawn Probability=&RandomRangeFloat(Minimum=0.1, Maximum=0.4)
    SpawnGroup=3
    SpawnRate=&Multiply_Float(A=15, B=$System.DelayFadeOut)
    Use Spawn Probability=true
   ) {module=SpawnRate001}
   &SpawnBurst_Instantaneous(
    Spawn Count=&RandomRangeInt(Minimum=2, Maximum=3)
    Spawn Group=3
    Spawn Time=$System.Delay
    Use Spawn Probability=false
    UseLoopCountLimit=false
   ) {module=SpawnBurst_Instantaneous003}
   &SpawnBurst_Instantaneous(Spawn Group=4, Use Spawn Probability=false, UseLoopCountLimit=false) {module=SpawnBurst_Instantaneous004}
   &SpawnBurst_Instantaneous(Spawn Count=5, Spawn Group=5, Use Spawn Probability=false, UseLoopCountLimit=false) {module=SpawnBurst_Instantaneous005}
   &SpawnRate(
    SpawnGroup=5
    SpawnRate=&FloatFromCurve(
     FloatCurve=DataInterface() {propertyValues {Curve {keys [
         {arriveTangent=-0.0405, interpMode=RCIM_Cubic, leaveTangent=-0.0405, tangentMode=RCTM_User, time=0, value=1.025}
         {arriveTangent=-34.9156, interpMode=RCIM_Cubic, leaveTangent=-34.9156, tangentMode=RCTM_User, time=0.1, value=0}
        ]}}}
     CurveIndex=$Emitter.NormalizedLoopAge
     Scale Curve=25
    )
    Use Spawn Probability=false
   ) {module=SpawnRate002}
  ]
  particleSpawnScript [
   /SetVariables(Particles.VisibilityTag=&SpawnGroup()) {isSetParameters=true}
   &InitializeParticle(
    Color=&MaskLinearColorBySpawnGroup(
     Masked Linear Color=&Multiply_LinearColorByFloat(Linear Color=$System.ElimColor, Float=0.5)
     Passthrough Linear Color=&MaskLinearColorBySpawnGroup(
      Masked Linear Color=$System.ElimColor
      Passthrough Linear Color=&MaskLinearColorBySpawnGroup(
       Masked Linear Color=&MultiplyLinearColor(
        Linear Color=$System.ElimColor
        Scale Factor=&MakeVector4_FromFloat(Float=&RandomRangeFloat(Minimum=0.8, Maximum=10))
       )
       Passthrough Linear Color=&MaskLinearColorBySpawnGroup(
        Masked Linear Color=$System.ElimColor
        Passthrough Linear Color=&MaskLinearColorBySpawnGroup(
         Masked Linear Color=$System.ElimColor
         Passthrough Linear Color=&Multiply_LinearColorByFloat(Linear Color=$System.ElimColor, Float=8)
         Spawn Group Mask=4
        )
        Spawn Group Mask=3
       )
       Spawn Group Mask=2
      )
      Spawn Group Mask=1
     )
    )
    Lifetime=&MaskFloatBySpawnGroup(
     Masked Float=2.4
     Passthrough Float=&MaskFloatBySpawnGroup(
      Masked Float=2.4
      Passthrough Float=&MaskFloatBySpawnGroup(
       Masked Float=&RandomRangeFloat(Minimum=0.175, Maximum=0.5)
       Passthrough Float=&MaskFloatBySpawnGroup(
        Masked Float=&RandomRangeFloat(Minimum=0.175, Maximum=0.5)
        Passthrough Float=&MaskFloatBySpawnGroup(Masked Float=0.8, Passthrough Float=&RandomRangeFloat(Minimum=0.2, Maximum=0.6), Spawn Group Mask=4)
        Spawn Group Mask=3
       )
       Spawn Group Mask=2
      )
      Spawn Group Mask=1
     )
    )
    Mass=&MaskFloatBySpawnGroup(
     Masked Float=0
     Passthrough Float=&MaskFloatBySpawnGroup(
      Masked Float=0
      Passthrough Float=&MaskFloatBySpawnGroup(Masked Float=0, Passthrough Float=&RandomRangeFloat(Minimum=0.75, Maximum=2), Spawn Group Mask=2)
      Spawn Group Mask=1
     )
    )
    Mass Mode=::NewEnumerator1
    Mesh Scale Mode=::NewEnumerator3
    Mesh Uniform Scale=&MaskFloatBySpawnGroup(
     Masked Float=0
     Passthrough Float=&MaskFloatBySpawnGroup(
      Masked Float=1
      Passthrough Float=&MaskFloatBySpawnGroup(
       Masked Float=&RandomRangeFloat(Minimum=0.5, Maximum=0.11)
       Passthrough Float=&MaskFloatBySpawnGroup(
        Masked Float=&RandomRangeFloat(Minimum=0.1, Maximum=0.4)
        Passthrough Float=&MaskFloatBySpawnGroup(
         Masked Float=0
         Passthrough Float=&MakeFloatFromVector(Vector=<0.05 0.05 0.05>, Channel=::NewEnumerator0)
         Spawn Group Mask=4
        )
        Spawn Group Mask=3
       )
       Spawn Group Mask=2
      )
      Spawn Group Mask=1
     )
    )
    Sprite Size Mode=::NewEnumerator3
    Uniform Sprite Size=&MaskFloatBySpawnGroup(
     Masked Float=130
     Passthrough Float=&MaskFloatBySpawnGroup(
      Masked Float=0
      Passthrough Float=&MaskFloatBySpawnGroup(Masked Float=0, Passthrough Float=&RandomRangeFloat(Minimum=1, Maximum=7), Spawn Group Mask=2)
      Spawn Group Mask=1
     )
    )
    UsePositionOffset=false
    Write Material Random=true
   ) {module=InitializeParticle001}
   &CameraOffset(Camera Offset Amount=10)
   /SetVariables(Particles.BacklightAlignment:Vector3f=<0 0 1>, Particles.BacklightCO=-15) {isSetParameters=true}
   &SpriteFacingAndAlignment(WriteAlignment=false, WriteFacing=true)
   &ShapeLocation(
    Handle Radius:NiagaraFloat=&MaskFloatBySpawnGroup(
     Masked Float=0
     Passthrough Float=&MaskFloatBySpawnGroup(
      Masked Float=0
      Passthrough Float=&MaskFloatBySpawnGroup(
       Masked Float=15
       Passthrough Float=&MaskFloatBySpawnGroup(
        Masked Float=15
        Passthrough Float=&MaskFloatBySpawnGroup(Masked Float=0, Passthrough Float=1, Spawn Group Mask=4)
        Spawn Group Mask=3
       )
       Spawn Group Mask=2
      )
      Spawn Group Mask=1
     )
    )
    Large Radius=&MaskFloatBySpawnGroup(
     Masked Float=0
     Passthrough Float=&MaskFloatBySpawnGroup(
      Masked Float=0
      Passthrough Float=&MaskFloatBySpawnGroup(
       Masked Float=40
       Passthrough Float=&MaskFloatBySpawnGroup(
        Masked Float=40
        Passthrough Float=&MaskFloatBySpawnGroup(Masked Float=0, Passthrough Float=1, Spawn Group Mask=4)
        Spawn Group Mask=3
       )
       Spawn Group Mask=2
      )
      Spawn Group Mask=1
     )
    )
    MaskSpawnGroup=false
    Override Seed=false
    Rotation Axis=&MaskVectorBySpawnGroup(
     Masked Vector=<0 0 0>
     Passthrough Vector=&MaskVectorBySpawnGroup(
      Masked Vector=<0 0 0>
      Passthrough Vector=&MaskVectorBySpawnGroup(
       Masked Vector=<1 0 0>
       Passthrough Vector=&MaskVectorBySpawnGroup(
        Masked Vector=<0 0 0>
        Passthrough Vector=&MaskVectorBySpawnGroup(Masked Vector=<0 0 0>, Passthrough Vector=<1 0 0>, Spawn Group Mask=4)
        Spawn Group Mask=3
       )
       Spawn Group Mask=2
      )
      Spawn Group Mask=1
     )
    )
    Rotation Mode=::NewEnumerator1
    Shape Primitive=::NewEnumerator3
    Surface Distribution=1
    UseOwnerScale=true
   )
   &InitialMeshOrientation(
    Mesh Axis To Orient=&MaskVectorBySpawnGroup(
     Masked Vector=<0 0 0>
     Passthrough Vector=&MaskVectorBySpawnGroup(
      Masked Vector=<0 1 0>
      Passthrough Vector=&RandomRangeVector(
       Minimum=<0 0 0>
       Maximum=&MaskVectorBySpawnGroup(
        Masked Vector=<0 0 0>
        Passthrough Vector=&MaskVectorBySpawnGroup(
         Masked Vector=<0 0 0>
         Passthrough Vector=&MaskVectorBySpawnGroup(Masked Vector=<0 0 0>, Passthrough Vector=<1 0 0>, Spawn Group Mask=4)
         Spawn Group Mask=3
        )
        Spawn Group Mask=2
       )
      )
      Spawn Group Mask=1
     )
    )
    Mesh Orientation Mode=::NewEnumerator2
    Orientation Vector=&MaskVectorBySpawnGroup(
     Masked Vector=<1 0 0>
     Passthrough Vector=&MaskVectorBySpawnGroup(
      Masked Vector=<1 0 0>
      Passthrough Vector=&MaskVectorBySpawnGroup(
       Masked Vector=<1 0 0>
       Passthrough Vector=&MaskVectorBySpawnGroup(
        Masked Vector=<1 0 0>
        Passthrough Vector=&MaskVectorBySpawnGroup(Masked Vector=<1 0 0>, Passthrough Vector=$Particles.ShapeLocation.ShapeVector, Spawn Group Mask=4)
        Spawn Group Mask=3
       )
       Spawn Group Mask=2
      )
      Spawn Group Mask=1
     )
    )
    Rotation=&MaskVectorBySpawnGroup(
     Masked Vector=<0 0 0>
     Passthrough Vector=&MaskVectorBySpawnGroup(
      Masked Vector=<0 0 0>
      Passthrough Vector=&MaskVectorBySpawnGroup(
       Masked Vector=&RandomRangeVector(Minimum=<-1 -1 -1>, Maximum=<1 1 1>)
       Passthrough Vector=&MaskVectorBySpawnGroup(
        Masked Vector=<0 0 0>
        Passthrough Vector=&MaskVectorBySpawnGroup(Masked Vector=<0 0 0>, Passthrough Vector=<0 0.3 0>, Spawn Group Mask=4)
        Spawn Group Mask=3
       )
       Spawn Group Mask=2
      )
      Spawn Group Mask=1
     )
    )
    Use Rotation Vector=true
   )
   &AddVelocity(
    Cone Angle=&MaskFloatBySpawnGroup(
     Masked Float=0
     Passthrough Float=&MaskFloatBySpawnGroup(
      Masked Float=0
      Passthrough Float=&MaskFloatBySpawnGroup(Masked Float=0, Passthrough Float=45, Spawn Group Mask=2)
      Spawn Group Mask=1
     )
    )
    Cone Axis=$Particles.ShapeLocation.ShapeNormal
    Distribution Along Cone Axis:NiagaraFloat=&MaskFloatBySpawnGroup(
     Masked Float=0
     Passthrough Float=&MaskFloatBySpawnGroup(
      Masked Float=0
      Passthrough Float=&MaskFloatBySpawnGroup(Masked Float=0, Passthrough Float=0.2, Spawn Group Mask=2)
      Spawn Group Mask=1
     )
    )
    Module.UseConeFalloff=false
    Module.UseVelDistribution=true
    Override Seed=false
    Velocity Mode=::NewEnumerator2
    Velocity Speed=&MaskFloatBySpawnGroup(
     Masked Float=0
     Passthrough Float=&MaskFloatBySpawnGroup(
      Masked Float=0
      Passthrough Float=&MaskFloatBySpawnGroup(
       Masked Float=0
       Passthrough Float=&MaskFloatBySpawnGroup(
        Masked Float=&RandomRangeFloat(Minimum=25, Maximum=500)
        Passthrough Float=&MaskFloatBySpawnGroup(Masked Float=0, Passthrough Float=0, Spawn Group Mask=4)
        Spawn Group Mask=3
       )
       Spawn Group Mask=2
      )
      Spawn Group Mask=1
     )
    )
   )
   &ApplyInitialForces()
   /SetVariables(Particles.InitialFlareAlignment:Vector3f=<0 0 1>) {isSetParameters=true}
   /SetVariables() {isSetParameters=true}
  ]
  particleUpdateScript [
   &ParticleState()
   &GravityForce(Gravity=&MaskVectorBySpawnGroup(
     Masked Vector=<0 0 0>
     Passthrough Vector=&MaskVectorBySpawnGroup(
      Masked Vector=<0 0 0>
      Passthrough Vector=&MaskVectorBySpawnGroup(
       Masked Vector=<0 0 0>
       Passthrough Vector=&MaskVectorBySpawnGroup(
        Masked Vector=<0 0 -980>
        Passthrough Vector=&MaskVectorBySpawnGroup(Masked Vector=<0 0 0>, Passthrough Vector=<0 0 0>, Spawn Group Mask=4)
        Spawn Group Mask=3
       )
       Spawn Group Mask=2
      )
      Spawn Group Mask=1
     )
    ))
   &Drag(
    Drag=&MaskFloatBySpawnGroup(
     Masked Float=0
     Passthrough Float=&MaskFloatBySpawnGroup(
      Masked Float=0
      Passthrough Float=&MaskFloatBySpawnGroup(
       Masked Float=0
       Passthrough Float=&FloatFromCurve(
        FloatCurve=DataInterface() {propertyValues {Curve {keys [{interpMode=RCIM_Cubic, tangentMode=RCTM_User, time=0, value=0}, {interpMode=RCIM_Cubic, tangentMode=RCTM_User, time=1, value=1}]}}}
        Scale Curve=&RandomRangeFloat(Minimum=0.8, Maximum=1.2)
       )
       Spawn Group Mask=2
      )
      Spawn Group Mask=1
     )
    )
    Use Linear Drag=true
    Use Rotational Drag=false
   )
   &ScaleMeshSize(Scale Factor=&MaskVectorBySpawnGroup(
     Masked Vector=<0 0 0>
     Passthrough Vector=&MaskVectorBySpawnGroup(
      Masked Vector=&VectorFromFloat(Value=&FloatFromCurve(
        FloatCurve=DataInterface() {propertyValues {Curve {keys [
            {arriveTangent=8.2579, interpMode=RCIM_Cubic, leaveTangent=6.1559, tangentMode=RCTM_Break, time=0, value=0.85}
            {arriveTangent=0.3, interpMode=RCIM_Cubic, leaveTangent=0.3, time=0.0698, value=1.1}
            {arriveTangent=-0.0863, interpMode=RCIM_Cubic, leaveTangent=-0.0863, tangentMode=RCTM_User, time=1, value=1.15}
           ]}}}
        Scale Curve=0.4
       ))
      Passthrough Vector=&MaskVectorBySpawnGroup(
       Masked Vector=&Multiply_VectorByFloat(
        Vector=&RandomRangeVector(Minimum=<1 1 2>, Maximum=<2 2 2>)
        Float=&FloatFromCurve(
         FloatCurve=DataInterface() {propertyValues {Curve {keys [{arriveTangent=0.6299, interpMode=RCIM_Cubic, leaveTangent=0.6299, tangentMode=RCTM_User, time=0, value=1}, <at 1 1.5>]}}}
         Scale Curve=&RandomRangeFloat(Minimum=0.01, Maximum=0.05)
        )
       )
       Passthrough Vector=&MaskVectorBySpawnGroup(
        Masked Vector=<1 1 1>
        Passthrough Vector=&MaskVectorBySpawnGroup(
         Masked Vector=<1 1 1>
         Passthrough Vector=&RandomRangeVector(Minimum=<1 1 1>, Maximum=<2 2 3>)
         Spawn Group Mask=4
        )
        Spawn Group Mask=3
       )
       Spawn Group Mask=2
      )
      Spawn Group Mask=1
     )
    ))
   &ScaleColor(
    Scale Alpha=&MaskFloatBySpawnGroup(
     Masked Float=&FloatFromCurve(FloatCurve=DataInterface() {propertyValues {Curve {keys [
          {arriveTangent=7.5976, interpMode=RCIM_Cubic, leaveTangent=7.5976, tangentMode=RCTM_User, time=0, value=0}
          {arriveTangent=-0.9298, interpMode=RCIM_Cubic, leaveTangent=-3.2722, tangentMode=RCTM_Break, time=0.5, value=1}
          {arriveTangent=-2, interpMode=RCIM_Cubic, leaveTangent=-0.1359, tangentMode=RCTM_Break, time=0.8, value=0.05}
          {tangentMode=RCTM_User, time=1, value=0}
         ]}}})
     Passthrough Float=&MaskFloatBySpawnGroup(
      Masked Float=&FloatFromCurve(FloatCurve=DataInterface() {propertyValues {Curve {keys [
           <at 0 0.75>
           <at 0.1 1>
           {arriveTangent=-3.1346, interpMode=RCIM_Cubic, leaveTangent=-3.1346, tangentMode=RCTM_User, time=0.5, value=0.75}
           {arriveTangent=-0.1427, tangentMode=RCTM_User, time=1, value=0}
          ]}}})
      Passthrough Float=&MaskFloatBySpawnGroup(
       Masked Float=&FloatFromCurve(FloatCurve=DataInterface() {propertyValues {Curve {keys [<at 0 1>,<at 0.1 3>,<at 0.2 1>,<at 1 1>]}}}, Scale Curve=2)
       Passthrough Float=&MaskFloatBySpawnGroup(
        Masked Float=1
        Passthrough Float=&MaskFloatBySpawnGroup(
         Masked Float=1
         Passthrough Float=&FloatFromCurve(FloatCurve=DataInterface() {propertyValues {Curve {keys [
              {arriveTangent=9.0725, interpMode=RCIM_Cubic, leaveTangent=9.0725, tangentMode=RCTM_User, time=0, value=0}
              {arriveTangent=-0.9508, interpMode=RCIM_Cubic, leaveTangent=-0.9508, tangentMode=RCTM_User, time=0.2, value=0.35}
              {arriveTangent=-0.0405, interpMode=RCIM_Cubic, leaveTangent=-0.0405, tangentMode=RCTM_User, time=0.9, value=0}
             ]}}})
         Spawn Group Mask=4
        )
        Spawn Group Mask=3
       )
       Spawn Group Mask=2
      )
      Spawn Group Mask=1
     )
    )
    Scale RGB=&MaskVectorBySpawnGroup(
     Masked Vector=&Lerp_Vector(
      A=&VectorFromFloat(Value=&FloatFromCurve(FloatCurve=DataInterface() {propertyValues {Curve {keys [<at 0 1>,<at 0.1 1>,<at 0.2 2>,<at 0.3 1>,<at 0.6 1>,<at 0.7 2>,<at 0.8 1>,<at 1 1>]}}}))
      B=&VectorFromFloat(Value=&RandomRangeFloat(Minimum=0.3, Maximum=0.4, Evaluation Type=::NewEnumerator1))
      Alpha=&FloatFromCurve(FloatCurve=DataInterface() {propertyValues {Curve {keys [
           {interpMode=RCIM_Cubic, leaveTangent=2.5294, tangentMode=RCTM_Break, time=0, value=0}
           {arriveTangent=-0.2166, interpMode=RCIM_Cubic, tangentMode=RCTM_Break, time=0.4, value=1}
          ]}}})
     )
     Passthrough Vector=&MaskVectorBySpawnGroup(
      Masked Vector=&VectorFromFloat(Value=&FloatFromCurve(
        FloatCurve=DataInterface() {propertyValues {Curve {keys [
            {arriveTangent=9.0725, interpMode=RCIM_Cubic, leaveTangent=9.0725, tangentMode=RCTM_User, time=0, value=1}
            {arriveTangent=8.5055, interpMode=RCIM_Cubic, leaveTangent=55.036, tangentMode=RCTM_Break, time=0.0565, value=15}
            {arriveTangent=-69.0971, interpMode=RCIM_Cubic, leaveTangent=-69.0971, tangentMode=RCTM_User, time=0.1066, value=12.5}
            {arriveTangent=-3.065, tangentMode=RCTM_User, time=0.4888, value=0.657}
            {arriveTangent=13.6988, interpMode=RCIM_Cubic, leaveTangent=13.6988, tangentMode=RCTM_User, time=1, value=0}
           ]}}}
        Scale Curve=3
       ))
      Passthrough Vector=&MaskVectorBySpawnGroup(
       Masked Vector=&Lerp_Vector(
        A=&VectorFromFloat(Value=&FloatFromCurve(
          FloatCurve=DataInterface() {propertyValues {Curve {keys [<at 0 1>,<at 0.1 1>,<at 0.2 2>,<at 0.3 1>,<at 0.6 1>,<at 0.7 2>,<at 0.8 1>,<at 1 1>]}}}
          Scale Curve=25
         ))
        B=&VectorFromFloat(Value=&RandomRangeFloat(Minimum=0.3, Maximum=5, Evaluation Type=::NewEnumerator1))
        Alpha=&FloatFromCurve(FloatCurve=DataInterface() {propertyValues {Curve {keys [
             {interpMode=RCIM_Cubic, leaveTangent=2.5294, tangentMode=RCTM_Break, time=0, value=0}
             {arriveTangent=-0.2166, interpMode=RCIM_Cubic, tangentMode=RCTM_Break, time=0.4, value=1}
            ]}}})
       )
       Passthrough Vector=&MaskVectorBySpawnGroup(
        Masked Vector=<1 1 1>
        Passthrough Vector=&MaskVectorBySpawnGroup(
         Masked Vector=<1 1 1>
         Passthrough Vector=&MaskVectorBySpawnGroup(
          Masked Vector=&VectorFromFloat(Value=&FloatFromCurve(FloatCurve=DataInterface() {propertyValues {Curve {keys [
                <at 0 1 RCIM_Cubic>
                {arriveTangent=-3.3132, interpMode=RCIM_Cubic, leaveTangent=-3.3132, tangentMode=RCTM_User, time=0.1, value=2}
                <at 1 1>
               ]}}}))
          Passthrough Vector=&VectorFromFloat(Value=&FloatFromCurve(FloatCurve=DataInterface() {propertyValues {Curve {keys [
                {arriveTangent=9.0725, interpMode=RCIM_Cubic, leaveTangent=9.0725, tangentMode=RCTM_User, time=0, value=0}
                {arriveTangent=269.3106, interpMode=RCIM_Cubic, leaveTangent=-89.8683, tangentMode=RCTM_Break, time=0.1, value=15}
                {arriveTangent=-0.819, interpMode=RCIM_Cubic, leaveTangent=-4.1834, tangentMode=RCTM_Break, time=0.5, value=1}
                {arriveTangent=2.9176, interpMode=RCIM_Cubic, leaveTangent=2.9176, tangentMode=RCTM_User, time=1, value=0}
               ]}}}))
         )
         Spawn Group Mask=4
        )
        Spawn Group Mask=3
       )
       Spawn Group Mask=2
      )
      Spawn Group Mask=1
     )
    )
    ScaleA=true
    ScaleRGB=true
   )
   &ScaleSpriteSize(
    Scale Sprite Size Mode=::NewEnumerator0
    Uniform Scale Factor=&MaskFloatBySpawnGroup(
     Masked Float=&FloatFromCurve(FloatCurve=DataInterface() {propertyValues {Curve {keys [
          {arriveTangent=4.4871, interpMode=RCIM_Cubic, leaveTangent=4.4871, tangentMode=RCTM_User, time=0, value=0.85}
          {arriveTangent=-0.0292, interpMode=RCIM_Cubic, leaveTangent=-0.0292, tangentMode=RCTM_User, time=0.2, value=1.15}
          {arriveTangent=-0.0863, interpMode=RCIM_Cubic, leaveTangent=-0.0863, tangentMode=RCTM_User, time=1, value=1}
         ]}}})
     Passthrough Float=&MaskFloatBySpawnGroup(
      Masked Float=1
      Passthrough Float=&MaskFloatBySpawnGroup(Masked Float=1, Passthrough Float=1, Spawn Group Mask=2)
      Spawn Group Mask=1
     )
    )
   )
   /SetVariables(
    Particles.BacklightColor:LinearColor=&MultiplyLinearColor(Linear Color=$Particles.Color, Scale Factor=&MakeVector4_FromLinearColor(LinearColor=<rgba 1 1 1 0.5>))
    Particles.BacklightSpriteSize:Vector2f=&Multiply_Vector2D(A=$Particles.SpriteSize, B=<2 2>)
    Particles.LensFlareColor:LinearColor=&MultiplyLinearColor(Linear Color=$Particles.Color, Scale Factor=&MakeVector4_FromLinearColor(LinearColor=<rgba 1 1 1 0.5>))
    Particles.LensFlareSize:Vector2f=&Multiply_Vector2D(A=$Particles.SpriteSize, B=<1.1 1.1>)
   ) {isSetParameters=true}
   /SetVariables(Particles.MedallionDMP:Vector4f=&MakeVector4(
     X=&FloatFromCurve(FloatCurve=DataInterface() {propertyValues {Curve {keys [
          {arriveTangent=0.9852, interpMode=RCIM_Cubic, leaveTangent=0.9852, tangentMode=RCTM_User, time=0, value=0}
          {arriveTangent=4.2284, interpMode=RCIM_Cubic, leaveTangent=3.5167, tangentMode=RCTM_Break, time=0.181, value=1}
          {arriveTangent=0.2227, interpMode=RCIM_Cubic, leaveTangent=-0.5782, tangentMode=RCTM_Break, time=0.3747, value=1.15}
          {arriveTangent=-1.7592, interpMode=RCIM_Cubic, leaveTangent=-5.0385, tangentMode=RCTM_Break, time=0.5197, value=0.95}
          {arriveTangent=-0.2575, interpMode=RCIM_Cubic, leaveTangent=-0.2575, tangentMode=RCTM_User, time=1, value=-0.2}
         ]}}})
     Y=4
     Z=1.5
     W=&FloatFromCurve(FloatCurve=DataInterface() {propertyValues {Curve {keys [{arriveTangent=0.2212, interpMode=RCIM_Cubic, leaveTangent=0.2212, tangentMode=RCTM_User, time=0, value=0}, <at 1 0>]}}})
    )) {isSetParameters=true}
   &DynamicMaterialParameters(
    Index 0 Param 1=&MakeFloatFromVector(Vector=$Particles.Scale, Channel=::NewEnumerator0)
    Index 0 Param 2=&RandomRangeFloat(Minimum=0, Maximum=35)
    Index 0 Param 3=&FloatFromCurve(
     FloatCurve=DataInterface() {propertyValues {Curve {keys [{interpMode=RCIM_Cubic, leaveTangent=8.1085, tangentMode=RCTM_Break, time=0.8568, value=1}, {tangentMode=RCTM_User, time=1, value=0}]}}}
     Scale Curve=&RandomRangeFloat(Minimum=0.07, Maximum=0.1)
    )
    Index 0 Param 4=&RandomRangeFloat(Minimum=0.8, Maximum=1)
    Index 1 Param 1=&FloatFromCurve(FloatCurve=DataInterface() {propertyValues {Curve {keys [
         {interpMode=RCIM_Cubic, leaveTangent=1.1034, tangentMode=RCTM_Break, time=0, value=0}
         {arriveTangent=2.2399, tangentMode=RCTM_User, time=0.5, value=1}
        ]}}})
    Index 1 Param 2=&FloatFromCurve(FloatCurve=DataInterface() {propertyValues {Curve {keys [{tangentMode=RCTM_User, time=0, value=0}, {tangentMode=RCTM_User, time=1, value=1}]}}})
    Index 1 Param 3=&FloatFromCurve(FloatCurve=DataInterface() {propertyValues {Curve {keys [
         {interpMode=RCIM_Cubic, tangentMode=RCTM_User, time=0, value=0}
         {arriveTangent=2.7163, interpMode=RCIM_Cubic, leaveTangent=2.7163, tangentMode=RCTM_User, time=1, value=1}
        ]}}})
    Index 1 Param 4=&FloatFromCurve(
     FloatCurve=DataInterface() {propertyValues {Curve {keys [
         {interpMode=RCIM_Cubic, leaveTangent=0.1169, tangentMode=RCTM_Break, time=0, value=0}
         {arriveTangent=2.5708, interpMode=RCIM_Cubic, leaveTangent=2.5708, tangentMode=RCTM_User, time=1, value=1}
        ]}}}
     Scale Curve=&RandomRangeFloat(Minimum=1, Maximum=5)
    )
    Index 2 Param 1=&FloatFromCurve(
     FloatCurve=DataInterface() {propertyValues {Curve {keys [
         {interpMode=RCIM_Cubic, leaveTangent=-0.0025, tangentMode=RCTM_Break, time=0, value=0.5}
         {arriveTangent=-0.0316, interpMode=RCIM_Cubic, tangentMode=RCTM_Break, time=1, value=-0.2}
        ]}}}
     Scale Curve=-0.5
    )
    Index 2 Param 2=&FloatFromCurve(FloatCurve=DataInterface() {propertyValues {Curve {keys [
         {interpMode=RCIM_Cubic, tangentMode=RCTM_User, time=0, value=0}
         {arriveTangent=2.0522, interpMode=RCIM_Cubic, leaveTangent=2.0522, tangentMode=RCTM_User, time=1, value=1}
        ]}}})
    Module.Index0Param0Write=true
    Module.Index0Param1Write=true
    Module.Index0Param2Write=true
    Module.Index0Param3Write=true
    Module.Index1Param0Write=true
    Module.Index1Param1Write=true
    Module.Index1Param2Write=true
    Module.Index1Param3Write=true
    Module.Index2Param0Write=true
    Module.Index2Param1Write=true
    Module.Index2Param2Write=false
    Module.Index2Param3Write=false
    Module.Index3Param0Write=false
    Module.Index3Param1Write=false
    Module.Index3Param2Write=false
    Module.Index3Param3Write=true
    Write Parameter Index 1=true
    Write Parameter Index 2=true
    Write Parameter Index 3=true
   ) {module=DynamicMaterialParameters001}
   &SolveForcesAndVelocity(Clamp Velocity=false, Limit Acceleration=false)
   &ScaleSpriteSizeBySpeed(
    Max Scale Factor=&MaskVector2dBySpawnGroup(
     Masked Vector2D=<0 0>
     Passthrough Vector2D=&MaskVector2dBySpawnGroup(
      Masked Vector2D=<0 0>
      Passthrough Vector2D=&MaskVector2dBySpawnGroup(Masked Vector2D=<0 0>, Passthrough Vector2D=<0.5 5>, Spawn Group Mask=2)
      Spawn Group Mask=1
     )
     Spawn Group Mask=0
    )
    Min Scale Factor=&MaskVector2dBySpawnGroup(
     Masked Vector2D=<0 0>
     Passthrough Vector2D=&MaskVector2dBySpawnGroup(
      Masked Vector2D=<0 0>
      Passthrough Vector2D=&MaskVector2dBySpawnGroup(Masked Vector2D=<0 0>, Passthrough Vector2D=<1 0.75>, Spawn Group Mask=2)
      Spawn Group Mask=1
     )
     Spawn Group Mask=0
    )
    Sample Scale Factor By Curve=true
    Velocity Threshold=500
   )
   /SetVariables(Particles.LensFlareDMP:Vector4f=&MakeVector4(X=0, Y=0, Z=0, W=0)) {isSetParameters=true}
   /SetVariables(
    Particles.InitialFlareColor:LinearColor=&MultiplyLinearColor(Linear Color=<rgba 15 15 15 1>, Scale Factor=&MakeVector4_FromLinearColor(LinearColor=$System.ElimColor))
    Particles.InitialFlareSize:Vector2f=&Multiply_Vector2D_ByFloat(
     Float=&FloatFromCurve(FloatCurve=DataInterface() {propertyValues {Curve {keys [
          {arriveTangent=7.0231, interpMode=RCIM_Cubic, leaveTangent=7.0231, tangentMode=RCTM_User, time=0, value=0}
          {arriveTangent=-2.5249, interpMode=RCIM_Cubic, leaveTangent=-2.5249, tangentMode=RCTM_User, time=0.2, value=0.95}
          {arriveTangent=-0.0405, interpMode=RCIM_Cubic, leaveTangent=-0.0405, tangentMode=RCTM_User, time=1, value=0}
         ]}}})
     Vector2D=<90 90>
    )
    Particles.SpriteDMP:Vector4f=&MakeVector4(X=0, Y=0, Z=0, W=0)
   ) {isSetParameters=true}
   /SetVariables(
    Particles.LightningDMP:Vector4f=&MakeVector4(
     X=&MakeFloatFromVector(Vector=$Particles.Scale, Channel=::NewEnumerator0)
     Y=&RandomRangeFloat(Minimum=0, Maximum=10)
     Z=&FloatFromCurve(
      FloatCurve=DataInterface() {propertyValues {Curve {keys [{interpMode=RCIM_Cubic, leaveTangent=8.1085, tangentMode=RCTM_Break, time=0.8568, value=1}, {tangentMode=RCTM_User, time=1, value=0}]}}}
      Scale Curve=&RandomRangeFloat(Minimum=0.09, Maximum=0.1)
     )
     W=0.8
    )
    Particles.LightningDMP1:Vector4f=&MakeVector4(
     X=&FloatFromCurve(FloatCurve=DataInterface() {propertyValues {Curve {keys [
          {interpMode=RCIM_Cubic, leaveTangent=1.1034, tangentMode=RCTM_Break, time=0, value=0}
          {arriveTangent=2.2399, tangentMode=RCTM_User, time=0.5, value=1}
         ]}}})
     Y=&FloatFromCurve(FloatCurve=DataInterface() {propertyValues {Curve {keys [<at 0.5 0>, {arriveTangent=3.2053, tangentMode=RCTM_User, time=0.9, value=1}]}}})
     Z=&FloatFromCurve(FloatCurve=DataInterface() {propertyValues {Curve {keys [
          {interpMode=RCIM_Cubic, tangentMode=RCTM_User, time=0, value=0}
          {arriveTangent=1.7507, interpMode=RCIM_Cubic, leaveTangent=1.7507, tangentMode=RCTM_User, time=1, value=1}
         ]}}})
     W=&FloatFromCurve(
      FloatCurve=DataInterface() {propertyValues {Curve {keys [
          {interpMode=RCIM_Cubic, leaveTangent=0.1169, tangentMode=RCTM_Break, time=0, value=0}
          {arriveTangent=2.5708, interpMode=RCIM_Cubic, leaveTangent=2.5708, tangentMode=RCTM_User, time=1, value=1}
         ]}}}
      Scale Curve=&RandomRangeFloat(Minimum=1, Maximum=40)
     )
    )
    Particles.LightningDMP2:Vector4f=&MakeVector4(
     X=&FloatFromCurve(
      FloatCurve=DataInterface() {propertyValues {Curve {keys [
          {interpMode=RCIM_Cubic, leaveTangent=-0.0025, tangentMode=RCTM_Break, time=0, value=0.5}
          {arriveTangent=-0.0316, interpMode=RCIM_Cubic, tangentMode=RCTM_Break, time=1, value=-0.2}
         ]}}}
      Scale Curve=4
     )
     Y=&FloatFromCurve(
      FloatCurve=DataInterface() {propertyValues {Curve {keys [
          {interpMode=RCIM_Cubic, tangentMode=RCTM_User, time=0, value=0}
          {arriveTangent=2.0522, interpMode=RCIM_Cubic, leaveTangent=2.0522, tangentMode=RCTM_User, time=1, value=1}
         ]}}}
      Scale Curve=2
     )
     Z=0
     W=0
    )
    Particles.LightningDMP3:Vector4f=&MakeVector4(X=0, Y=0, Z=0, W=0)
   ) {isSetParameters=true}
   /SetVariables(Particles.LightningScale:Vector3f=&Multiply_Vector(
     A=&RandomRangeVector(Minimum=<1 1 1>, Maximum=<2 2 3>)
     B=&VectorFromCurve(
      VectorCurve=DataInterface() {propertyValues {
        XCurve {keys [{interpMode=RCIM_Cubic, tangentMode=RCTM_User, time=0, value=0}, {interpMode=RCIM_Cubic, tangentMode=RCTM_User, time=1, value=1}]}
        YCurve {keys [{interpMode=RCIM_Cubic, tangentMode=RCTM_User, time=0, value=0}, {interpMode=RCIM_Cubic, tangentMode=RCTM_User, time=1, value=1}]}
        ZCurve {keys [{tangentMode=RCTM_User, time=0, value=0}, {tangentMode=RCTM_User, time=1, value=1}]}
       }}
      Scale Curve=&RandomRangeVector(Minimum=<0.1 0.1 0.04>, Maximum=<0.15 0.15 0.09>)
     )
    )) {isSetParameters=true}
  ]
  rendererClasses [&NiagaraSpriteRendererProperties, &NiagaraMeshRendererProperties]
  renderers [
   {rendererClass=&NiagaraSpriteRendererProperties}
   {rendererClass=&NiagaraSpriteRendererProperties, rendererIndex=1}
   {rendererClass=&NiagaraMeshRendererProperties, rendererIndex=2}
   {rendererClass=&NiagaraMeshRendererProperties, rendererIndex=3}
   {rendererClass=&NiagaraMeshRendererProperties, rendererIndex=4}
   {rendererClass=&NiagaraSpriteRendererProperties, rendererIndex=5}
   {rendererClass=&NiagaraSpriteRendererProperties, rendererIndex=6}
   {rendererClass=&NiagaraMeshRendererProperties, rendererIndex=7}
   {rendererClass=&NiagaraMeshRendererProperties, rendererIndex=8}
  ]
 }}
summary {
 systemName=NS_Emote_TotalFlock_Medallion
 emitters [{emitterName=Medallion, rendererClasses [&NiagaraSpriteRendererProperties, &NiagaraMeshRendererProperties]}]
 userVariables [
  User.bIsFrontEndPreview {type=NiagaraBool}
  User.bKillStage {type=NiagaraInt32}
 ]
}
